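# Default lit, masked, two-sided Paper2D material for rendering sprites with scene lighting.
# Vertex color and alpha multiply the sprite's diffuse texture color and alpha, tinting the surface and its cutout.
# A second sprite sampler reads an additional texture slot as a normal map, defaulting to a flat placeholder.
# Scalar parameters expose Metallic and Roughness, defaulting to a nonmetal, fully matte surface.
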Material /Paper2D/DefaultLitSpriteMaterial
graph {
 properties {
  blendMode=BLEND_Masked
  pixelDepthOffsetMode=PDOM_Legacy
  twoSided=true
  usedShadingModels=Default Lit
  lightmassSettings {
   bCastShadowAsMasked=true
   bOverrideCastShadowAsMasked=false
   bOverrideDiffuseBoost=false
   bOverrideEmissiveBoost=false
   bOverrideExportResolutionScale=false
   diffuseBoost=1
   emissiveBoost=1
   exportResolutionScale=1
  }
 }
 nodes [
  n1 = VertexColor()
  n2 = TextureCoordinate()
  n3 = SpriteTextureSampler!(UVs=@n2, Apply View MipBias) {slotDisplayName=DiffuseMap, texture=/Paper2D/PlaceholderTextures/DummySpriteTexture}
  Multiply(A=@n1, B=@n3.RGB)
  Multiply(A=@n1.A, B=@n3.A)
  ScalarParameter() {desc="Nonmetals have Metallic values of 0, metals have Metallic values of 1.", parameterName=Metallic}
  ScalarParameter() {
   defaultValue=1
   desc="Roughness of 0 (smooth) is a mirror reflection and roughness of 1 (rough) is completely matte or diffuse. "
   parameterName=Roughness
  }
  SpriteTextureSampler(UVs=@n2) {
   bSampleAdditionalTextures=true
   parameterName=SpriteAdditionalTexture_0
   samplerType=SAMPLERTYPE_Normal
   slotDisplayName=NormalMap
   texture=/Paper2D/PlaceholderTextures/FlatNormalMap
  }
 ]
}
