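# Default Paper2D sprite material: unlit, masked and two-sided, and also flagged for use with Niagara sprites.
# A SpriteTextureSampler reads the sprite texture at the mesh UVs, with a dummy placeholder texture as its default.
# Texture RGB is multiplied by vertex color for tinting, and texture alpha by vertex alpha for masking.
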Material /Paper2D/DefaultSpriteMaterial
graph {
 properties {
  bUsedWithNiagaraSprites=true
  blendMode=BLEND_Masked
  pixelDepthOffsetMode=PDOM_Legacy
  shadingModel=MSM_Unlit
  twoSided=true
  usedShadingModels=Unlit
  lightmassSettings {
   bCastShadowAsMasked=true
   bOverrideCastShadowAsMasked=false
   bOverrideDiffuseBoost=false
   bOverrideEmissiveBoost=false
   bOverrideExportResolutionScale=false
   diffuseBoost=1
   emissiveBoost=1
   exportResolutionScale=1
  }
 }
 nodes [
  n1 = SpriteTextureSampler!(UVs=TextureCoordinate(), Apply View MipBias) {texture=/Paper2D/PlaceholderTextures/DummySpriteTexture}
  n2 = VertexColor()
  Multiply(A=@n1.A, B=@n2.A)
  Multiply(A=@n1.RGB, B=@n2)
 ]
}
